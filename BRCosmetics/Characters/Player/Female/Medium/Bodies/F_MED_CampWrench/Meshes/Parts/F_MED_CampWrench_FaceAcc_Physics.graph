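# Physics asset for the CampWrench female medium FaceAcc mesh, simulating hair and a side hair string against upper-body capsules.
# Center, left, right, two side chains and the string hang from head-attached base bones, mostly with narrow swing and twist limits.
# Angular-free constraints cross-link the center chain to left and right, and join the right side's mid-chain base and extra branch.
# Head, neck, spine, clavicles and upper arms form the body shapes, with duplicate right-side "mover" bodies.
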PhysicsAsset /BRCosmetics/Characters/Player/Female/Medium/Bodies/F_MED_CampWrench/Meshes/Parts/F_MED_CampWrench_FaceAcc_Physics
bodies {
 base_Hair_C_00 [{shapeType=Capsule, length=0.505, radius=0.505, shapeName=base_Hair_C_00_capsule, rotation=<rot 0 -90 0>}]
 base_Hair_L_00 [{shapeType=Capsule, length=0.505, radius=0.505, shapeName=base_Hair_L_00_capsule, rotation=<rot 0 -90 0>}]
 base_Hair_R_00 [{shapeType=Capsule, length=0.505, radius=0.505, shapeName=base_Hair_R_00_capsule, rotation=<rot 0 -90 0>}]
 base_Hair_Side_L_00 [{shapeType=Capsule, length=0.505, radius=0.505, shapeName=base_Hair_Side_L_00_capsule, rotation=<rot 0 -90 0>}]
 base_Hair_Side_R_00 [{shapeType=Capsule, length=0.505, radius=0.505, shapeName=base_Hair_Side_R_00_capsule, rotation=<rot 0 -90 0>}]
 base_Hair_Side_R_01 [{shapeType=Capsule, length=0.505, radius=0.505, shapeName=base_Hair_Side_R_01_capsule, rotation=<rot 0 -90 0>}]
 base_sideHairString_00 [{shapeType=Capsule, length=0.505, radius=0.505, shapeName=base_sideHairString_00_capsule, rotation=<rot 0 -90 0>}]
 clavicle_l [{shapeType=Capsule, length=4.1598, radius=6.5405, shapeName=clavicle_l_capsule, center=<5.2088 -0.9604 -2.6655>, rotation=<rot 90 0 0>}]
 clavicle_r [{
   shapeType=Capsule
   length=3.8998
   radius=6.8005
   shapeName=clavicle_l_capsule
   center=<-5.2125 0.9604 2.4239>
   rotation=<rot 88.5323 -179.9983 -179.9983>
  }]
 dyn_Hair_C_00 [{shapeType=Capsule, length=4.4512, radius=3.0099, shapeName=dyn_Hair_C_00_capsule, center=<3.9603 0 0>, rotation=<rot 90 0 0>}]
 dyn_Hair_C_01 [{shapeType=Capsule, length=4.4512, radius=3.0099, shapeName=dyn_Hair_C_01_capsule, center=<4.2019 -1.2818 0>, rotation=<rot 90 -5.4607 0>}]
 dyn_Hair_C_02 [{
   shapeType=Capsule
   length=1.5951
   radius=3.2006
   shapeName=dyn_Hair_C_02_capsule
   center=<1.3189 -2.3763 -0.264>
   rotation=<rot 90 -15.9995 0>
  }]
 dyn_Hair_L_00 [{shapeType=Capsule, length=4.4512, radius=3.0099, shapeName=dyn_Hair_C_01_capsule, center=<4.2311 0 0>, rotation=<rot 90 0 0>}]
 dyn_Hair_L_01 [{
   shapeType=Capsule
   length=4.4512
   radius=3.0099
   shapeName=dyn_Hair_C_01_capsule
   center=<4.1114 -0.3624 0.3014>
   rotation=<rot 49.7919 92.6003 108.3201>
  }]
 dyn_Hair_L_02 [{
   shapeType=Capsule
   length=4.4247
   radius=2.5148
   shapeName=dyn_Hair_L_02_capsule
   center=<0.6681 -0.9837 0.8251>
   rotation=<rot 14.829 18.5351 22.4006>
  }]
 dyn_Hair_R_00 [{shapeType=Capsule, length=4.4512, radius=3.0099, shapeName=dyn_Hair_C_01_capsule, center=<4.2311 0 0.1672>, rotation=<rot 90 0 0>}]
 dyn_Hair_R_01 [{
   shapeType=Capsule
   length=4.1901
   radius=3.0099
   shapeName=dyn_Hair_C_01_capsule
   center=<3.8381 -0.6542 -0.0796>
   rotation=<rot 54.0314 -89.9998 -79.6683>
  }]
 dyn_Hair_R_02 [
  {
   shapeType=Capsule
   length=3.4638
   radius=2.89
   shapeName=dyn_Hair_R_02_capsule
   center=<-0.0496 -1.4787 -0.4368>
   rotation=<rot -19.164 16.4556 -32.1555>
  }
  {
   shapeType=Capsule
   length=2.1122
   radius=2.3979
   shapeName=dyn_Hair_R_02_capsule
   center=<-1.7312 2.1607 -3.5105>
   rotation=<rot 15.7806 21.5687 55.4729>
  }
 ]
 dyn_Hair_Side_L_00 [{
   shapeType=Capsule
   length=5.292
   radius=2.4925
   shapeName=dyn_Hair_Side_L_00_capsule
   center=<3.5571 -0.5978 0.0098>
   rotation=<rot 80.3841 -89.9999 -90>
  }]
 dyn_Hair_Side_L_01 [{
   shapeType=Capsule
   length=2.9087
   radius=2.8492
   shapeName=dyn_Hair_Side_L_01_capsule
   center=<2.75 0.5969 0.1576>
   rotation=<rot 88.4155 -179.9998 -179.9999>
  }]
 dyn_Hair_Side_L_02 [
  {
   shapeType=Capsule
   length=1.9995
   radius=1.5674
   shapeName=dyn_Hair_Side_L_02_capsule
   center=<2.1552 -0.4145 2.6873>
   rotation=<rot 0 -6.6618 90>
  }
  {
   shapeType=Capsule
   length=2.6297
   radius=2.2221
   shapeName=dyn_Hair_Side_L_02_capsule
   center=<2.0438 1.7867 1.1371>
   rotation=<rot 1.3014 -2.0676 -61.5535>
  }
 ]
 dyn_Hair_Side_R_00 [{
   shapeType=Capsule
   length=4.3306
   radius=2.2278
   shapeName=dyn_Hair_Side_R_00_capsule
   center=<2.9781 0 -0.8981>
   rotation=<rot 87.8762 180 180>
  }]
 dyn_Hair_Side_R_01 [{
   shapeType=Capsule
   length=3.874
   radius=3.1236
   shapeName=dyn_Hair_Side_R_01_capsule
   center=<2.0002 -0.6608 -0.9777>
   rotation=<rot 87.889 -89.954 180>
  }]
 dyn_Hair_Side_R_02 [{shapeType=Capsule, length=4.5614, radius=1.8079, shapeName=dyn_Hair_Side_R_02_capsule, center=<2.3016 0.3961 0.44>, rotation=<rot 90 90 0>}]
 dyn_Hair_Side_R_03 [
  {
   shapeType=Capsule
   length=2.43
   radius=1.1043
   shapeName=dyn_Hair_Side_R_03_capsule
   center=<1.2516 -0.2942 -3.6943>
   rotation=<rot 90 77.0248 0>
  }
  {
   shapeType=Capsule
   length=2.8643
   radius=1.4714
   shapeName=dyn_Hair_Side_R_03_capsule
   center=<1.3852 -0.2704 -2.1769>
   rotation=<rot -80.646 63.2137 -166.9929>
  }
 ]
 dyn_Hair_Side_R_04 [{
   shapeType=Capsule
   length=4.3641
   radius=3.201
   shapeName=dyn_Hair_Side_R_04_capsule
   center=<1.2606 -0.0039 -0.3247>
   rotation=<rot 0 63.8727 -63.2867>
  }]
 dyn_Hair_Side_R_05 [{
   shapeType=Capsule
   length=2.6589
   radius=1.8927
   shapeName=dyn_Hair_Side_R_05_capsule
   center=<2.0308 0.3149 0.2985>
   rotation=<rot -22.1248 -90 -37.8697>
  }]
 dyn_Hair_Side_R_06 [{
   shapeType=Capsule
   length=3.3214
   radius=1.9457
   shapeName=dyn_Hair_Side_R_06_capsule
   center=<1.033 -0.2056 -1.1252>
   rotation=<rot 0 -130.3008 -61.7396>
  }]
 dyn_sideHairString_01 [{
   shapeType=Capsule
   length=2.3608
   radius=0.6392
   shapeName=dyn_sideHairString_01_capsule
   center=<2.4037 0.3218 -0.1328>
   rotation=<rot 57.7368 -45.6438 -57.1217>
  }]
 dyn_sideHairString_02 [{
   shapeType=Capsule
   length=3.6314
   radius=0.7085
   shapeName=dyn_sideHairString_02_capsule
   center=<1.8159 0.4006 -0.822>
   rotation=<rot 22.8636 87.869 71.144>
  }]
 dyn_sideHairString_03 [{
   shapeType=Capsule
   length=1.8696
   radius=1.1304
   shapeName=dyn_sideHairString_03_capsule
   center=<1.0569 -0.741 1.1652>
   rotation=<rot 40.9484 175.0574 -157.3312>
  }]
 head [{shapeType=Capsule, length=3.1204, radius=5.0198, shapeName=head_capsule, center=<-0.0278 2.785 -0.122>, rotation=<rot 77.1613 -0.0001 0>}]
 mover_clavicle_r [{
   shapeType=Capsule
   length=3.8998
   radius=6.8005
   shapeName=clavicle_l_capsule
   center=<-5.2125 0.9604 2.4239>
   rotation=<rot 88.5323 -179.9983 -179.9983>
  }]
 mover_upperarm_r [{
   shapeType=Capsule
   length=10.6329
   radius=4.9532
   shapeName=upperarm_l_capsule
   center=<-5.4593 0.2245 0.7534>
   rotation=<rot 85.0289 -2.5806 -0.0016>
  }]
 neck_02 [{shapeType=Capsule, length=5.1387, radius=4.2818, shapeName=neck_02_capsule, center=<-2.3824 0.1489 0>, rotation=<rot 90 0 0>}]
 spine_05 [{shapeType=Capsule, length=12.5292, radius=7.808, shapeName=spine_04_capsule, center=<1.7412 -0.6684 0>, rotation=<rot 0 -92.6847 0>}]
 upperarm_l [{
   shapeType=Capsule
   length=7.8623
   radius=3.568
   shapeName=upperarm_l_capsule
   center=<3.7053 -0.1455 -0.6007>
   rotation=<rot 85.0289 -2.5806 -0.0016>
  }]
 upperarm_r [{
   shapeType=Capsule
   length=10.6329
   radius=4.9532
   shapeName=upperarm_l_capsule
   center=<-5.4593 0.2245 0.7534>
   rotation=<rot 85.0289 -2.5806 -0.0016>
  }]
}
constraints [
 {
  bone1Name=head
  bone2Name=neck_02
  swing1LimitDegrees=45
  swing1Motion=Free
  swing2LimitDegrees=45
  swing2Motion=Free
  twistLimitDegrees=45
  twistMotion=Free
 }
 {
  bone1Name=base_Hair_C_00
  bone2Name=head
  swing1LimitDegrees=45
  swing1Motion=Free
  swing2LimitDegrees=45
  swing2Motion=Free
  twistLimitDegrees=45
  twistMotion=Free
 }
 {bone1Name=dyn_Hair_C_00, bone2Name=base_Hair_C_00, swing1LimitDegrees=13, swing2LimitDegrees=15, twistLimitDegrees=1}
 {bone1Name=dyn_Hair_C_01, bone2Name=dyn_Hair_C_00, swing1LimitDegrees=30, swing2LimitDegrees=20, twistLimitDegrees=1}
 {bone1Name=dyn_Hair_C_02, bone2Name=dyn_Hair_C_01, swing1LimitDegrees=45, swing2LimitDegrees=10, twistLimitDegrees=1}
 {
  bone1Name=base_Hair_L_00
  bone2Name=head
  swing1LimitDegrees=45
  swing1Motion=Free
  swing2LimitDegrees=45
  swing2Motion=Free
  twistLimitDegrees=45
  twistMotion=Free
 }
 {bone1Name=dyn_Hair_L_00, bone2Name=base_Hair_L_00, swing1LimitDegrees=10, swing2LimitDegrees=13, twistLimitDegrees=1}
 {bone1Name=dyn_Hair_L_01, bone2Name=dyn_Hair_L_00, swing1LimitDegrees=20, swing2LimitDegrees=13, twistLimitDegrees=1}
 {bone1Name=dyn_Hair_L_02, bone2Name=dyn_Hair_L_01, swing1LimitDegrees=20, swing2LimitDegrees=10, twistLimitDegrees=1}
 {
  bone1Name=base_Hair_R_00
  bone2Name=head
  swing1LimitDegrees=45
  swing1Motion=Free
  swing2LimitDegrees=45
  swing2Motion=Free
  twistLimitDegrees=45
  twistMotion=Free
 }
 {bone1Name=dyn_Hair_R_00, bone2Name=base_Hair_R_00, swing1LimitDegrees=13, swing2LimitDegrees=13, twistLimitDegrees=1}
 {bone1Name=dyn_Hair_R_01, bone2Name=dyn_Hair_R_00, swing1LimitDegrees=20, swing2LimitDegrees=13, twistLimitDegrees=1}
 {bone1Name=dyn_Hair_R_02, bone2Name=dyn_Hair_R_01, swing1LimitDegrees=40, swing2LimitDegrees=10, twistLimitDegrees=1}
 {
  bone1Name=base_Hair_Side_L_00
  bone2Name=head
  swing1LimitDegrees=45
  swing1Motion=Free
  swing2LimitDegrees=45
  swing2Motion=Free
  twistLimitDegrees=45
  twistMotion=Free
 }
 {bone1Name=dyn_Hair_Side_L_00, bone2Name=base_Hair_Side_L_00, swing1LimitDegrees=5, swing2LimitDegrees=10, twistLimitDegrees=1}
 {bone1Name=dyn_Hair_Side_L_01, bone2Name=dyn_Hair_Side_L_00, swing1LimitDegrees=12, swing2LimitDegrees=15, twistLimitDegrees=1}
 {bone1Name=dyn_Hair_Side_L_02, bone2Name=dyn_Hair_Side_L_01, swing1LimitDegrees=10, swing2LimitDegrees=20, twistLimitDegrees=1}
 {
  bone1Name=base_Hair_Side_R_00
  bone2Name=head
  swing1LimitDegrees=45
  swing1Motion=Free
  swing2LimitDegrees=45
  swing2Motion=Free
  twistLimitDegrees=45
  twistMotion=Free
 }
 {bone1Name=dyn_Hair_Side_R_00, bone2Name=base_Hair_Side_R_00, swing1LimitDegrees=5, swing2LimitDegrees=5, twistLimitDegrees=1}
 {bone1Name=dyn_Hair_Side_R_01, bone2Name=base_Hair_Side_R_01, swing1LimitDegrees=5, swing2LimitDegrees=16, twistLimitDegrees=1}
 {bone1Name=dyn_Hair_Side_R_02, bone2Name=dyn_Hair_Side_R_01, swing1LimitDegrees=10, swing2LimitDegrees=10, twistLimitDegrees=1}
 {bone1Name=dyn_Hair_Side_R_03, bone2Name=dyn_Hair_Side_R_02, swing1LimitDegrees=5, swing2LimitDegrees=20, twistLimitDegrees=1}
 {
  bone1Name=base_sideHairString_00
  bone2Name=head
  swing1LimitDegrees=45
  swing1Motion=Free
  swing2LimitDegrees=45
  swing2Motion=Free
  twistLimitDegrees=45
  twistMotion=Free
 }
 {bone1Name=dyn_sideHairString_01, bone2Name=base_sideHairString_00, swing1LimitDegrees=15, swing2LimitDegrees=20, twistLimitDegrees=1}
 {bone1Name=dyn_sideHairString_02, bone2Name=dyn_sideHairString_01, swing1LimitDegrees=20, swing2LimitDegrees=20, twistLimitDegrees=1}
 {bone1Name=dyn_sideHairString_03, bone2Name=dyn_sideHairString_02, swing1LimitDegrees=25, swing2LimitDegrees=25, twistLimitDegrees=1}
 {
  bone1Name=upperarm_l
  bone2Name=clavicle_l
  swing1LimitDegrees=45
  swing1Motion=Free
  swing2LimitDegrees=45
  swing2Motion=Free
  twistLimitDegrees=45
  twistMotion=Free
 }
 {
  bone1Name=upperarm_r
  bone2Name=clavicle_r
  swing1LimitDegrees=45
  swing1Motion=Free
  swing2LimitDegrees=45
  swing2Motion=Free
  twistLimitDegrees=45
  twistMotion=Free
 }
 {
  bone1Name=neck_02
  bone2Name=spine_05
  swing1LimitDegrees=45
  swing1Motion=Free
  swing2LimitDegrees=45
  swing2Motion=Free
  twistLimitDegrees=45
  twistMotion=Free
 }
 {
  bone1Name=clavicle_l
  bone2Name=spine_05
  swing1LimitDegrees=45
  swing1Motion=Free
  swing2LimitDegrees=45
  swing2Motion=Free
  twistLimitDegrees=45
  twistMotion=Free
 }
 {
  bone1Name=clavicle_r
  bone2Name=spine_05
  swing1LimitDegrees=45
  swing1Motion=Free
  swing2LimitDegrees=45
  swing2Motion=Free
  twistLimitDegrees=45
  twistMotion=Free
 }
 {
  bone1Name=dyn_Hair_R_01
  bone2Name=dyn_Hair_C_01
  swing1LimitDegrees=45
  swing1Motion=Free
  swing2LimitDegrees=45
  swing2Motion=Free
  twistLimitDegrees=45
  twistMotion=Free
 }
 {
  bone1Name=dyn_Hair_L_01
  bone2Name=dyn_Hair_C_01
  swing1LimitDegrees=45
  swing1Motion=Free
  swing2LimitDegrees=45
  swing2Motion=Free
  twistLimitDegrees=45
  twistMotion=Free
 }
 {
  bone1Name=mover_upperarm_r
  bone2Name=upperarm_r
  swing1LimitDegrees=45
  swing1Motion=Free
  swing2LimitDegrees=45
  swing2Motion=Free
  twistLimitDegrees=45
  twistMotion=Free
 }
 {
  bone1Name=mover_clavicle_r
  bone2Name=clavicle_r
  swing1LimitDegrees=45
  swing1Motion=Free
  swing2LimitDegrees=45
  swing2Motion=Free
  twistLimitDegrees=45
  twistMotion=Free
 }
 {
  bone1Name=base_Hair_Side_R_01
  bone2Name=dyn_Hair_Side_R_00
  swing1LimitDegrees=45
  swing1Motion=Free
  swing2LimitDegrees=45
  swing2Motion=Free
  twistLimitDegrees=45
  twistMotion=Free
 }
 {
  bone1Name=dyn_Hair_Side_R_06
  bone2Name=dyn_Hair_Side_R_05
  swing1LimitDegrees=45
  swing1Motion=Free
  swing2LimitDegrees=45
  swing2Motion=Free
  twistLimitDegrees=45
  twistMotion=Free
 }
 {
  bone1Name=dyn_Hair_Side_R_05
  bone2Name=dyn_Hair_Side_R_04
  swing1LimitDegrees=45
  swing1Motion=Free
  swing2LimitDegrees=45
  swing2Motion=Free
  twistLimitDegrees=45
  twistMotion=Free
 }
 {
  bone1Name=dyn_Hair_Side_R_04
  bone2Name=dyn_Hair_Side_R_01
  swing1LimitDegrees=45
  swing1Motion=Free
  swing2LimitDegrees=45
  swing2Motion=Free
  twistLimitDegrees=45
  twistMotion=Free
 }
]
