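# Animation Blueprint for the CampWrench female character's face accessory mesh, driving its hair motion.
# The parent mesh pose feeds a Control Rig with emote, targeting and pickaxe flags, then additive bone offsets.
# RigidBody simulation runs in spine_05 space with gravity reduced during skydive floating, followed by head, neck, spine and clavicle constraints and corrective offsets.
# Backbling and cape flags are set on customization load but unread by the AnimGraph.
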AnimBlueprint /BRCosmetics/Characters/Player/Female/Medium/Bodies/F_MED_CampWrench/Meshes/Parts/F_MED_CampWrench_FaceAcc_AnimBP
target_skeleton=&/F_MED_CampWrench_FaceAcc_Skeleton
graphs [
 AnimGraph {nodes [
   Output Pose(Result=Use cached pose 'Hair Rigid Body'()) {layerGroup=None}
   n1 = Control Rig!(
    Source=Copy Pose From Mesh() {bUseAttachedParent=true}
    Alpha
    Playing Emote:bool=Get bIsPlayingEmote()
    Is Targeting:bool=Get bIsTargeting()
    is Pickaxe Swing:bool=Get is Pickaxe swing() -> (is Pickaxe swing:bool)
   ) -> (Pose) {
    alphaBoolBlend {blendOption=Linear}
    controlRigAssetReference {blueprintRigClass=&/F_MED_CampWrench_FaceAcc_CtrlRig.F_MED_CampWrench_FaceAcc_CtrlRig_C}
   }
   Save cached pose 'CR Pose'(Pose=@n1) {cacheName=CR Pose}
   n2 = "Transform (Modify) Bone - Bone: base_HairMaster"(ComponentPose=Local To Component(LocalPose=Use cached pose 'CR Pose'()), Alpha=1, Translation="0,0,0", Rotation="0,0,0", Scale="1,1,1") {
    rotationMode=BMM_Additive
    rotationSpace=BCS_BoneSpace
    translationMode=BMM_Additive
    translationSpace=BCS_BoneSpace
    alphaBoolBlend {blendOption=Linear}
    alphaScaleBiasClamp {bInterpResult=true}
    boneToModify {boneName=base_HairMaster}
   }
   n3 = "Transform (Modify) Bone - Bone: base_Hair_R_00"(ComponentPose=@n2, Alpha=1, Translation="0,0,0", Rotation="0,-10,0", Scale="1,1,1") {
    rotationMode=BMM_Additive
    rotationSpace=BCS_BoneSpace
    translationMode=BMM_Additive
    translationSpace=BCS_BoneSpace
    alphaBoolBlend {blendOption=Linear}
    alphaScaleBiasClamp {bInterpResult=true}
    boneToModify {boneName=base_Hair_R_00}
    rotation {yaw=-10}
   }
   n4 = "Transform (Modify) Bone - Bone: base_Hair_L_00"(ComponentPose=@n3, Alpha=1, Translation="0,0,0", Rotation="0,-10,0", Scale="1,1,1") {
    rotationMode=BMM_Additive
    rotationSpace=BCS_BoneSpace
    translationMode=BMM_Additive
    translationSpace=BCS_BoneSpace
    alphaBoolBlend {blendOption=Linear}
    alphaScaleBiasClamp {bInterpResult=true}
    boneToModify {boneName=base_Hair_L_00}
    rotation {yaw=-10}
   }
   n5 = "Transform (Modify) Bone - Bone: base_HairMaster"(ComponentPose=@n4, Alpha=1, Translation="0,0,0", Rotation="0,15,0", Scale="1,1,1") {
    rotationMode=BMM_Additive
    rotationSpace=BCS_BoneSpace
    translationMode=BMM_Additive
    translationSpace=BCS_BoneSpace
    alphaBoolBlend {blendOption=Linear}
    alphaScaleBiasClamp {bInterpResult=true}
    boneToModify {boneName=base_HairMaster}
    rotation {yaw=15}
   }
   n6 = "Transform (Modify) Bone - Bone: base_Hair_R_00"(ComponentPose=@n5, Alpha=1, Translation="0,0,0", Rotation="0,5,0", Scale="1,1,1") {
    rotationMode=BMM_Additive
    rotationSpace=BCS_BoneSpace
    translationMode=BMM_Additive
    translationSpace=BCS_BoneSpace
    alphaBoolBlend {blendOption=Linear}
    alphaScaleBiasClamp {bInterpResult=true}
    boneToModify {boneName=base_Hair_R_00}
    rotation {yaw=5}
   }
   n7 = "Transform (Modify) Bone - Bone: base_Hair_L_00"(ComponentPose=@n6, Alpha=1, Translation="0,0,0", Rotation="0,-7,0", Scale="1,1,1") {
    rotationMode=BMM_Additive
    rotationSpace=BCS_BoneSpace
    translationMode=BMM_Additive
    translationSpace=BCS_BoneSpace
    alphaBoolBlend {blendOption=Linear}
    alphaScaleBiasClamp {bInterpResult=true}
    boneToModify {boneName=base_Hair_L_00}
    rotation {yaw=-7}
   }
   n8 = RigidBody!(ComponentPose=@n7, Alpha, OverrideWorldGravity:Vector=Get Override World Gravity() -> (Override World Gravity:Vector), ExternalForce) -> (Pose) {
    bClampLinearTranslationLimitToRefPose=true
    bOverrideWorldGravity=true
    overridePhysicsAsset=&/F_MED_CampWrench_FaceAcc_Physics
    simulationSpace=BaseBoneSpace
    alphaBoolBlend {blendOption=Linear}
    baseBoneRef {boneName=spine_05}
    simSpaceSettings {
     dampingAlpha=0
     maxAngularAcceleration=16
     maxAngularVelocity=8
     maxLinearAcceleration=1000
     maxLinearVelocity=500
     worldAlpha=1
     externalLinearDragV=<0.2 0.2>
    }
   }
   n9 = "Transform (Modify) Bone - Bone: dyn_Hair_Side_R_04"(ComponentPose=@n8, Alpha=1, Translation="0,0,0", Rotation="0,0,-35", Scale="1,1,1") {
    rotationMode=BMM_Additive
    rotationSpace=BCS_BoneSpace
    alphaBoolBlend {blendOption=Linear}
    boneToModify {boneName=dyn_Hair_Side_R_04}
    rotation {roll=-35}
   }
   n10 = "Constraint - head_corrective (PP)"!(ComponentPose=@n9, Alpha=1, ConstraintWeights_0=1, ConstraintWeights_1:float=0) -> (Pose) {
    alphaBoolBlend {blendOption=Linear}
    boneToModify {boneName=head_corrective}
    constraintSetup [{targetBone {boneName=head}}, {targetBone {boneName=spine_05}}]
    constraintWeights [1,1]
   }
   n11 = "Constraint - dyn_Hair_C_00 (PP)"!(ComponentPose=@n10, Alpha=1, ConstraintWeights_0=1, ConstraintWeights_1:float=0) -> (Pose) {
    alphaBoolBlend {blendOption=Linear}
    boneToModify {boneName=dyn_Hair_C_00}
    constraintSetup [{targetBone {boneName=head}}, {targetBone {boneName=spine_05}}]
    constraintWeights [1,1]
   }
   n12 = "Constraint - dyn_Hair_C_01 (PPP)"!(ComponentPose=@n11, Alpha=1, ConstraintWeights_0=0, ConstraintWeights_1:float=0.5, ConstraintWeights_2:float=0.5) -> (Pose) {
    alphaBoolBlend {blendOption=Linear}
    boneToModify {boneName=dyn_Hair_C_01}
    constraintSetup [{targetBone {boneName=head}}, {targetBone {boneName=spine_05}}, {targetBone {boneName=neck_01}}]
    constraintWeights [1,1,1]
   }
   n13 = "Constraint - dyn_Hair_C_02 (PP)"!(ComponentPose=@n12, Alpha=1, ConstraintWeights_0=0, ConstraintWeights_1:float=1) -> (Pose) {
    alphaBoolBlend {blendOption=Linear}
    boneToModify {boneName=dyn_Hair_C_02}
    constraintSetup [{targetBone {boneName=head}}, {targetBone {boneName=spine_05}}]
    constraintWeights [1,1]
   }
   n14 = "Constraint - dyn_Hair_R_00 (PP)"!(ComponentPose=@n13, Alpha=1, ConstraintWeights_0=1, ConstraintWeights_1:float=0) -> (Pose) {
    alphaBoolBlend {blendOption=Linear}
    boneToModify {boneName=dyn_Hair_R_00}
    constraintSetup [{targetBone {boneName=head}}, {targetBone {boneName=spine_05}}]
    constraintWeights [1,1]
   }
   n15 = "Constraint - dyn_Hair_R_01 (PP)"!(ComponentPose=@n14, Alpha=1, ConstraintWeights_0=0.5, ConstraintWeights_1:float=0.5) -> (Pose) {
    alphaBoolBlend {blendOption=Linear}
    boneToModify {boneName=dyn_Hair_R_01}
    constraintSetup [{targetBone {boneName=neck_01}}, {targetBone {boneName=spine_05}}]
    constraintWeights [1,1]
   }
   n16 = "Constraint - dyn_Hair_R_02 (PP)"!(ComponentPose=@n15, Alpha=1, ConstraintWeights_0=0.7, ConstraintWeights_1:float=0.3) -> (Pose) {
    alphaBoolBlend {blendOption=Linear}
    boneToModify {boneName=dyn_Hair_R_02}
    constraintSetup [{targetBone {boneName=clavicle_r}}, {targetBone {boneName=spine_05}}]
    constraintWeights [1,1]
   }
   n17 = "Constraint - dyn_Hair_L_00 (P)"(ComponentPose=@n16, Alpha=1, ConstraintWeights_0=1) {
    alphaBoolBlend {blendOption=Linear}
    boneToModify {boneName=dyn_Hair_L_00}
    constraintSetup [{targetBone {boneName=head}}]
    constraintWeights [1]
   }
   n18 = "Constraint - dyn_Hair_L_01 (PP)"!(ComponentPose=@n17, Alpha=1, ConstraintWeights_0=0.5, ConstraintWeights_1:float=0.5) -> (Pose) {
    alphaBoolBlend {blendOption=Linear}
    boneToModify {boneName=dyn_Hair_L_01}
    constraintSetup [{targetBone {boneName=neck_01}}, {targetBone {boneName=spine_05}}]
    constraintWeights [1,1]
   }
   n19 = "Constraint - dyn_Hair_L_02 (PP)"!(ComponentPose=@n18, Alpha=1, ConstraintWeights_0=0.5, ConstraintWeights_1:float=0.5) -> (Pose) {
    alphaBoolBlend {blendOption=Linear}
    boneToModify {boneName=dyn_Hair_L_02}
    constraintSetup [{targetBone {boneName=clavicle_l}}, {targetBone {boneName=spine_05}}]
    constraintWeights [1,1]
   }
   n20 = "Transform (Modify) Bone - Bone: dyn_Hair_R_01"(ComponentPose=@n19, Alpha=1, Translation="0,-2,0", Rotation="-20,0,0", Scale="1,1,1") {
    rotationMode=BMM_Additive
    rotationSpace=BCS_BoneSpace
    scaleMode=BMM_Additive
    scaleSpace=BCS_BoneSpace
    translationMode=BMM_Additive
    translationSpace=BCS_BoneSpace
    alphaBoolBlend {blendOption=Linear}
    alphaScaleBiasClamp {bInterpResult=true}
    boneToModify {boneName=dyn_Hair_R_01}
    rotation {pitch=-20}
    translation {y=-2}
   }
   n21 = "Transform (Modify) Bone - Bone: dyn_Hair_R_02"(ComponentPose=@n20, Alpha=1, Translation="0,-2,0", Rotation="0,0,0", Scale="1,1,1") {
    rotationMode=BMM_Additive
    rotationSpace=BCS_BoneSpace
    scaleMode=BMM_Additive
    scaleSpace=BCS_BoneSpace
    translationMode=BMM_Additive
    translationSpace=BCS_BoneSpace
    alphaBoolBlend {blendOption=Linear}
    alphaScaleBiasClamp {bInterpResult=true}
    boneToModify {boneName=dyn_Hair_R_02}
    translation {y=-2}
   }
   n22 = "Transform (Modify) Bone - Bone: dyn_Hair_L_01"(ComponentPose=@n21, Alpha=1, Translation="0,-1,0", Rotation="20,0,0", Scale="1,1,1") {
    rotationMode=BMM_Additive
    rotationSpace=BCS_BoneSpace
    scaleMode=BMM_Additive
    scaleSpace=BCS_BoneSpace
    translationMode=BMM_Additive
    translationSpace=BCS_BoneSpace
    alphaBoolBlend {blendOption=Linear}
    alphaScaleBiasClamp {bInterpResult=true}
    boneToModify {boneName=dyn_Hair_L_01}
    rotation {pitch=20}
    translation {y=-1}
   }
   n23 = "Transform (Modify) Bone - Bone: dyn_Hair_L_02"(ComponentPose=@n22, Alpha=1, Translation="0,-1,0", Rotation="0,0,0", Scale="1,1,1") {
    rotationMode=BMM_Additive
    rotationSpace=BCS_BoneSpace
    scaleMode=BMM_Additive
    scaleSpace=BCS_BoneSpace
    translationMode=BMM_Additive
    translationSpace=BCS_BoneSpace
    alphaBoolBlend {blendOption=Linear}
    alphaScaleBiasClamp {bInterpResult=true}
    boneToModify {boneName=dyn_Hair_L_02}
    translation {y=-1}
   }
   Save cached pose 'Hair Rigid Body'(Pose=Component To Local(ComponentPose=@n23)) {cacheName=Hair Rigid Body}
  ]}
 EventGraph {nodes [
   Event BlueprintUpdateAnimation()
   > Set is Pickaxe swing(is Pickaxe swing:bool=Get bIsPlayingMeleeAnim()) -> (Output_Get:bool)
   > n1 = Branch(Condition=Get bIsSkydiveFloating())
   > Set Override World Gravity(Override World Gravity:Vector="0,0,-300") -> (Output_Get:Vector)
   Set Override World Gravity(execute=@n1.else, Override World Gravity="0,0,-980")
   Event OnCharacterCustomizationLoaded()
   n2 = CheckForBackbling()
   > Set is Wearing Backbling(is Wearing Backbling:bool=AND Boolean!(A=@n2, B=NOT Boolean(A=@n2.bIsBackblingACape), C:bool=NOT Boolean(A=@n2.bIsBackblingLong)) -> (ReturnValue)) -> (Output_Get:bool)
   > Set is Wearing Cape(is Wearing Cape:bool=OR Boolean(A=@n2.bIsBackblingACape, B=@n2.bIsBackblingLong)) -> (Output_Get:bool)
   > Set is Wearing Backbling or Cape(is Wearing Backbling or Cape:bool=OR Boolean(A=Get is Wearing Backbling() -> (is Wearing Backbling:bool), B=Get is Wearing Cape() -> (is Wearing Cape:bool))) -> (Output_Get:bool)
   Event BlueprintInitializeAnimation()
   > Set Skel Mesh(Skel Mesh=GetOwnerSkeletalMeshComponent())
  ]}
]
variables [
 is Pickaxe swing {type=boolean}
 is Wearing Backbling {type=boolean}
 is Wearing Backbling or Cape {type=boolean}
 is Wearing Cape {type=boolean}
 Override World Gravity {type=Vector, default=<0 0 -980>}
 Skel Mesh {type=SkeletalMeshComponent}
]
